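# Physics asset for the cape of the female medium WiseBarn player body, defining capsule bodies and joints for cape motion.
# Left and right four-segment chains and a three-segment middle chain hang from tiny base bodies linked to spine_04.
# Swing limits mostly widen down each chain, with twist locked on most joints.
# Spine, upper-arm and one left wing capsule add body volumes, linked by fully free angular constraints.
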PhysicsAsset /BRCosmetics/Characters/Player/Female/Medium/Bodies/M_MED_WiseBarn/Meshes/F_MED_WiseBarn_Cape_Physics1
bodies {
 base_cape_l [{shapeType=Capsule, length=0.505, radius=0.505, shapeName=base_cape_l_capsule, rotation=<rot 0 -90 0>}]
 base_cape_mid [{shapeType=Capsule, length=0.505, radius=0.505, shapeName=base_cape_mid_capsule, rotation=<rot 0 -90 0>}]
 base_cape_r [{shapeType=Capsule, length=0.505, radius=0.505, shapeName=base_cape_r_capsule, rotation=<rot 0 -90 0>}]
 dfrm_upperarm_l [{
   shapeType=Capsule
   length=8.6883
   radius=4.5
   shapeName=dfrm_upperarm_r_capsule
   center=<18.9998 -5.3525 -5.6323>
   rotation=<rot -9.9999 165.0004 40.0003>
  }]
 dfrm_upperarm_r [{shapeType=Capsule, length=8.6883, radius=4.5, shapeName=dfrm_upperarm_r_capsule, center=<-19 5 6.7>, rotation=<rot -10 165.0002 15.0001>}]
 dyn_cape_l_01 [{shapeType=Capsule, length=15, radius=3, shapeName=dyn_cape_l_01_capsule, center=<6.6452 0 0>, rotation=<rot 89.9079 179.9998 179.9998>}]
 dyn_cape_l_02 [{shapeType=Capsule, length=5.6965, radius=5.3035, shapeName=dyn_cape_l_02_capsule, center=<6 0 0>, rotation=<rot 90 0 0>}]
 dyn_cape_l_03 [{shapeType=Capsule, length=16, radius=9.6, shapeName=dyn_cape_r_02_capsule, center=<15 2 4>, rotation=<rot 85 0 0>}]
 dyn_cape_l_04 [{shapeType=Capsule, length=11, radius=4, shapeName=dyn_cape_l_04_capsule, center=<6 0 0>, rotation=<rot 0 90 90>}]
 dyn_cape_mid_01 [{shapeType=Capsule, length=7.8068, radius=4.4729, shapeName=dyn_cape_mid_01_capsule, center=<6.6197 0 0>, rotation=<rot 84.908 -180 -180>}]
 dyn_cape_mid_02 [{shapeType=Capsule, length=11, radius=6.6, shapeName=dyn_cape_mid_02_capsule, center=<11 0 0>, rotation=<rot 90 0 0>}]
 dyn_cape_mid_03 [{shapeType=Capsule, length=16, radius=9.6, shapeName=dyn_cape_mid_03_capsule, center=<15 0 0>, rotation=<rot 85 0 0>}]
 dyn_cape_r_01 [{shapeType=Capsule, length=10.636, radius=3, shapeName=dyn_cape_r_01_capsule, center=<6.6452 0 0>, rotation=<rot 89.9079 -179.9997 -179.9997>}]
 dyn_cape_r_02 [{shapeType=Capsule, length=11, radius=5, shapeName=dyn_cape_r_02_capsule, center=<11 0 0>, rotation=<rot 90 0 0>}]
 dyn_cape_r_03 [{shapeType=Capsule, length=16, radius=9, shapeName=dyn_cape_r_03_capsule, center=<15 -2 2>, rotation=<rot 90 0 0>}]
 dyn_cape_r_04 [{shapeType=Capsule, length=6, radius=4, shapeName=dyn_cape_r_04_capsule, center=<6 0 0>, rotation=<rot 0 90 90>}]
 dyn_wing_upper_l_02 [{shapeType=Capsule, length=22, radius=3, shapeName=dyn_wing_upper_l_02_capsule, center=<-6 2.5 -2.5>, rotation=<rot 0 90 100>}]
 spine_01 [{shapeType=Capsule, length=41.1229, radius=8, shapeName=spine_01_capsule, center=<-6.2 1 0>}]
 spine_02 [
  {shapeType=Capsule, length=33, radius=13, shapeName=spine_02_capsule, center=<-0.6554 3.8926 0>}
  {shapeType=Capsule, length=54.9008, radius=6.9827, shapeName=spine_02_capsule, center=<-2.3869 4.1232 0>}
 ]
 spine_04 [
  {shapeType=Capsule, length=14, radius=16.3195, shapeName=spine_04_capsule, center=<-7.0673 5.6584 0>}
  {shapeType=Capsule, length=54.4101, radius=4.2076, shapeName=spine_04_capsule, center=<-6.8988 0 0>}
 ]
}
constraints [
 {bone1Name=dyn_cape_l_01, bone2Name=base_cape_l, swing1LimitDegrees=0, swing2LimitDegrees=10, twistLimitDegrees=1, twistMotion=Locked}
 {bone1Name=dyn_cape_l_02, bone2Name=dyn_cape_l_01, swing1LimitDegrees=8, swing2LimitDegrees=30, twistLimitDegrees=1, twistMotion=Locked}
 {bone1Name=dyn_cape_l_03, bone2Name=dyn_cape_l_02, swing1LimitDegrees=15, swing2LimitDegrees=45, twistLimitDegrees=1, twistMotion=Locked}
 {bone1Name=dyn_cape_mid_01, bone2Name=base_cape_mid, swing1LimitDegrees=0, swing2LimitDegrees=10, twistLimitDegrees=1, twistMotion=Locked}
 {bone1Name=dyn_cape_mid_02, bone2Name=dyn_cape_mid_01, swing1LimitDegrees=8, swing2LimitDegrees=30, twistLimitDegrees=1, twistMotion=Locked}
 {bone1Name=dyn_cape_mid_03, bone2Name=dyn_cape_mid_02, swing1LimitDegrees=15, swing2LimitDegrees=45, twistLimitDegrees=1, twistMotion=Locked}
 {bone1Name=dyn_cape_r_01, bone2Name=base_cape_r, swing1LimitDegrees=0, swing2LimitDegrees=10, twistLimitDegrees=0}
 {bone1Name=dyn_cape_r_02, bone2Name=dyn_cape_r_01, swing1LimitDegrees=8, swing2LimitDegrees=30, twistLimitDegrees=1, twistMotion=Locked}
 {bone1Name=dyn_cape_r_03, bone2Name=dyn_cape_r_02, swing1LimitDegrees=15, swing2LimitDegrees=45, twistLimitDegrees=1, twistMotion=Locked}
 {
  bone1Name=base_cape_l
  bone2Name=spine_04
  swing1LimitDegrees=45
  swing1Motion=Free
  swing2LimitDegrees=45
  swing2Motion=Free
  twistLimitDegrees=45
  twistMotion=Free
 }
 {
  bone1Name=base_cape_r
  bone2Name=spine_04
  swing1LimitDegrees=45
  swing1Motion=Free
  swing2LimitDegrees=45
  swing2Motion=Free
  twistLimitDegrees=45
  twistMotion=Free
 }
 {
  bone1Name=base_cape_mid
  bone2Name=spine_04
  swing1LimitDegrees=45
  swing1Motion=Free
  swing2LimitDegrees=45
  swing2Motion=Free
  twistLimitDegrees=45
  twistMotion=Free
 }
 {
  bone1Name=spine_04
  bone2Name=spine_02
  swing1LimitDegrees=45
  swing1Motion=Free
  swing2LimitDegrees=45
  swing2Motion=Free
  twistLimitDegrees=45
  twistMotion=Free
 }
 {
  bone1Name=spine_02
  bone2Name=spine_01
  swing1LimitDegrees=45
  swing1Motion=Free
  swing2LimitDegrees=45
  swing2Motion=Free
  twistLimitDegrees=45
  twistMotion=Free
 }
 {
  bone1Name=dfrm_upperarm_r
  bone2Name=spine_04
  swing1LimitDegrees=45
  swing1Motion=Free
  swing2LimitDegrees=45
  swing2Motion=Free
  twistLimitDegrees=45
  twistMotion=Free
 }
 {
  bone1Name=dfrm_upperarm_l
  bone2Name=spine_04
  swing1LimitDegrees=45
  swing1Motion=Free
  swing2LimitDegrees=45
  swing2Motion=Free
  twistLimitDegrees=45
  twistMotion=Free
 }
 {bone1Name=dyn_cape_l_04, bone2Name=dyn_cape_l_03, swing1LimitDegrees=25, swing2LimitDegrees=25, twistLimitDegrees=1}
 {bone1Name=dyn_cape_r_04, bone2Name=dyn_cape_r_03, swing1LimitDegrees=25, swing2LimitDegrees=25, twistLimitDegrees=1}
 {
  bone1Name=dyn_wing_upper_l_02
  bone2Name=dfrm_upperarm_l
  swing1LimitDegrees=45
  swing1Motion=Free
  swing2LimitDegrees=45
  swing2Motion=Free
  twistLimitDegrees=45
  twistMotion=Free
 }
]
