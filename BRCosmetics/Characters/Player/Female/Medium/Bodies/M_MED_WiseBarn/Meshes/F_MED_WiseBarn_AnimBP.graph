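# Post-process style AnimBlueprint for the WiseBarn female cosmetic body, animating its wings, tail, cape and calf feathers.
# The pose copied from the attached parent passes through a Control Rig, additive bone corrections and a tail RigidBody.
# A backpack-switched cape branch (RigidBody alphas zero) then feeds spring-based AnimDynamics for wings and feathers, chained through cached poses.
# The EventGraph derives driving values from movement, weapon, emote, skydive and frontend states, plus gravity override.
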AnimBlueprint /BRCosmetics/Characters/Player/Female/Medium/Bodies/M_MED_WiseBarn/Meshes/F_MED_WiseBarn_AnimBP
target_skeleton=&/F_MED_WiseBarn_Skeleton
graphs [
 AnimGraph {nodes [
   Output Pose(Result=Use cached pose 'Dynamics_CACHE'()) {layerGroup=None}
   n1 = RigidBody!(ComponentPose=Local To Component(LocalPose=Use cached pose 'PreDynamics_CACHE'()), Alpha, OverrideWorldGravity:Vector="0,0,0", ExternalForce) -> (Pose) {
    bOverrideWorldGravity=true
    overridePhysicsAsset=&/F_MED_WiseBarn_Tail_Physics
    simulationSpace=BaseBoneSpace
    alphaBoolBlend {blendOption=Linear}
    baseBoneRef {boneName=pelvis}
    simSpaceSettings {
     maxAngularAcceleration=100
     maxAngularVelocity=3
     maxLinearAcceleration=700
     maxLinearVelocity=500
     velocityScaleZ=0
     worldAlpha=1
     externalLinearDragV=<0.2 0.2>
    }
   }
   n2 = "Transform (Modify) Bone - Bone: dyn_tail_MSTR"!(ComponentPose=@n1, bAlphaBoolEnabled:bool=True, Translation="0,0,0", Rotation="0,0,35", Scale="1,1,1") -> (Pose) {
    alphaInputType=Bool
    rotationMode=BMM_Additive
    alphaBoolBlend {blendInTime=0.5, blendOption=QuadraticInOut, blendOutTime=1}
    alphaScaleBiasClamp {bInterpResult=true}
    boneToModify {boneName=dyn_tail_MSTR}
    rotation {roll=35}
   }
   Save cached pose 'BaseLogic_CACHE'(Pose=Component To Local(ComponentPose=@n2)) {cacheName=BaseLogic_CACHE}
   n3 = Control Rig!(
    Source=Copy Pose From Mesh() {bUseAttachedParent=true}
    Alpha
    sprint:double=Get sprint() -> (sprint:double)
    emote:double=Get emote() -> (emote:double)
    pistol:double=Get pistol() -> (pistol:double)
    isCrouchingCR:bool=Get bIsCrouching()
    isTargetingCR:bool=Get bIsTargeting()
    WearingBackPackCR:bool=Get isWearingBackpack() -> (isWearingBackpack:bool)
    isFrontEndCR:bool=Get bIsInFrontEnd()
    TacticalSprintCR:bool=Get bIsTacticalSprinting()
    IsMeleePickaxe:bool=Get bIsPlayingMeleeAnim()
    IsSkydive:bool=Get bIsSkyDiving()
   ) -> (Pose) {
    alphaBoolBlend {blendOption=Linear}
    controlRigAssetReference {blueprintRigClass=&/F_MED_WiseBarn_CtrlRig.F_MED_WiseBarn_CtrlRig_C}
   }
   Save cached pose 'ControlRig_CACHE'(Pose=@n3) {cacheName=ControlRig_CACHE}
   n4 = Get bIsParachuteOpened()
   n5 = "Transform (Modify) Bone - Bone: dyn_wing_upper_l_twist_MSTR"!(
    ComponentPose=Local To Component(LocalPose=Use cached pose 'ControlRig_CACHE'())
    bAlphaBoolEnabled:bool=@n4
    Translation="0,0,0"
    Rotation="-0.010,0.039,15.000"
    Scale="1,1,1"
   ) -> (Pose) {
    alphaInputType=Bool
    rotationMode=BMM_Additive
    rotationSpace=BCS_BoneSpace
    alphaBoolBlend {blendInTime=0.1, blendOption=Linear, blendOutTime=0.1}
    boneToModify {boneName=dyn_wing_upper_l_twist_MSTR}
    rotation=<rot -0.01 0.039 15>
   }
   n6 = "Transform (Modify) Bone - Bone: dyn_wing_upper_r_twist_MSTR"!(ComponentPose=@n5, bAlphaBoolEnabled:bool=@n4, Translation="0,0,0", Rotation="0,0,15", Scale="1,1,1") -> (Pose) {
    alphaInputType=Bool
    rotationMode=BMM_Additive
    rotationSpace=BCS_BoneSpace
    alphaBoolBlend {blendInTime=0.1, blendOption=Linear, blendOutTime=0.1}
    boneToModify {boneName=dyn_wing_upper_r_twist_MSTR}
    rotation {roll=15}
   }
   n7 = Get frontend() -> (frontend:double)
   n8 = "Transform (Modify) Bone - Bone: dyn_wing_upper_l_twist_MSTR"!(ComponentPose=@n6, Alpha:float=@n7, Translation="0,0,0", Rotation="5,0,-5", Scale="1,1,1") -> (Pose) {
    rotationMode=BMM_Additive
    rotationSpace=BCS_BoneSpace
    alphaBoolBlend {blendInTime=0.1, blendOption=Linear, blendOutTime=0.1}
    boneToModify {boneName=dyn_wing_upper_l_twist_MSTR}
    rotation {pitch=5, roll=-5}
   }
   n9 = "Transform (Modify) Bone - Bone: dyn_wing_upper_r_twist_MSTR"!(ComponentPose=@n8, Alpha:float=@n7, Translation="0,0,0", Rotation="-5,0,5", Scale="1,1,1") -> (Pose) {
    rotationMode=BMM_Additive
    rotationSpace=BCS_BoneSpace
    alphaBoolBlend {blendInTime=0.1, blendOption=Linear, blendOutTime=0.1}
    boneToModify {boneName=dyn_wing_upper_r_twist_MSTR}
    rotation {pitch=-5, roll=5}
   }
   n10 = "Transform (Modify) Bone - Bone: dyn_wing_upper_l_MSTR"(ComponentPose=@n9, Alpha=@n7, Translation="0,-2,-1", Rotation="0,0,0", Scale="1,1,1") {
    rotationMode=BMM_Additive
    rotationSpace=BCS_BoneSpace
    translationMode=BMM_Additive
    translationSpace=BCS_BoneSpace
    alphaBoolBlend {blendInTime=0.1, blendOption=Linear, blendOutTime=0.1}
    boneToModify {boneName=dyn_wing_upper_l_MSTR}
    translation {y=-2, z=-1}
   }
   n11 = "Transform (Modify) Bone - Bone: dyn_wing_upper_r_MSTR"(ComponentPose=@n10, Alpha=@n7, Translation="0,2,1", Rotation="0,0,0", Scale="1,1,1") {
    rotationMode=BMM_Additive
    rotationSpace=BCS_BoneSpace
    translationMode=BMM_Additive
    translationSpace=BCS_BoneSpace
    alphaBoolBlend {blendInTime=0.1, blendOption=Linear, blendOutTime=0.1}
    boneToModify {boneName=dyn_wing_upper_r_MSTR}
    translation {y=2, z=1}
   }
   n12 = "Transform (Modify) Bone - Bone: dyn_tail_MSTR"!(ComponentPose=@n11, Alpha:float=1, Translation="-0.000,-0.725,7.545", Rotation="0,0,20", Scale="1,1,1") -> (Pose) {
    rotationMode=BMM_Additive
    translationMode=BMM_Additive
    alphaBoolBlend {blendOption=Linear}
    alphaScaleBiasClamp {bInterpResult=true}
    boneToModify {boneName=dyn_tail_MSTR}
    rotation {roll=20}
    translation {y=-0.725, z=7.545}
   }
   n13 = "Transform (Modify) Bone - Bone: dyn_tail_l_MSTR"!(ComponentPose=@n12, bAlphaBoolEnabled:bool=True, Translation="-10,-10,10", Rotation="0,-140,0", Scale="1,1,1") -> (Pose) {
    alphaInputType=Bool
    rotationMode=BMM_Additive
    rotationSpace=BCS_BoneSpace
    translationMode=BMM_Additive
    translationSpace=BCS_BoneSpace
    alphaBoolBlend {blendInTime=0.5, blendOutTime=1}
    boneToModify {boneName=dyn_tail_l_MSTR}
    rotation {yaw=-140}
    translation=<-10 -10 10>
   }
   n14 = "Transform (Modify) Bone - Bone: dyn_tail_MSTR"!(ComponentPose=@n13, bAlphaBoolEnabled:bool=True, Translation="-10,-15,0", Rotation="0,-60,0", Scale="1,1,1") -> (Pose) {
    alphaInputType=Bool
    rotationMode=BMM_Additive
    rotationSpace=BCS_BoneSpace
    translationMode=BMM_Additive
    translationSpace=BCS_BoneSpace
    alphaBoolBlend {blendInTime=0.5, blendOutTime=1}
    boneToModify {boneName=dyn_tail_MSTR}
    rotation {yaw=-60}
    translation=<-10 -15>
   }
   n15 = "Transform (Modify) Bone - Bone: dyn_tail_r_MSTR"!(ComponentPose=@n14, bAlphaBoolEnabled:bool=True, Translation="-10,-10,-5", Rotation="0,-140,0", Scale="1,1,1") -> (Pose) {
    alphaInputType=Bool
    rotationMode=BMM_Additive
    rotationSpace=BCS_BoneSpace
    translationMode=BMM_Additive
    translationSpace=BCS_BoneSpace
    alphaBoolBlend {blendInTime=0.5, blendOutTime=1}
    boneToModify {boneName=dyn_tail_r_MSTR}
    rotation {yaw=-140}
    translation=<-10 -10 -5>
   }
   n16 = "Transform (Modify) Bone - Bone: dyn_tail_l_MSTR"(ComponentPose=@n15, bAlphaBoolEnabled=True, Translation="0,-5,0", Rotation="0,-15,0", Scale="1,1,1") {
    alphaInputType=Bool
    rotationMode=BMM_Additive
    rotationSpace=BCS_BoneSpace
    translationMode=BMM_Additive
    translationSpace=BCS_BoneSpace
    alphaBoolBlend {blendInTime=0.5, blendOutTime=1}
    boneToModify {boneName=dyn_tail_l_MSTR}
    rotation {yaw=-15}
    translation {y=-5}
   }
   n17 = "Transform (Modify) Bone - Bone: dyn_tail_MSTR"(ComponentPose=@n16, bAlphaBoolEnabled=True, Translation="0,-5,0", Rotation="0,-20,0", Scale="1,1,1") {
    alphaInputType=Bool
    rotationMode=BMM_Additive
    rotationSpace=BCS_BoneSpace
    translationMode=BMM_Additive
    translationSpace=BCS_BoneSpace
    alphaBoolBlend {blendInTime=0.5, blendOutTime=1}
    boneToModify {boneName=dyn_tail_MSTR}
    rotation {yaw=-20}
    translation {y=-5}
   }
   n18 = "Transform (Modify) Bone - Bone: dyn_tail_r_MSTR"(ComponentPose=@n17, bAlphaBoolEnabled=True, Translation="0,-5,0", Rotation="0,-15,0", Scale="1,1,1") {
    alphaInputType=Bool
    rotationMode=BMM_Additive
    rotationSpace=BCS_BoneSpace
    translationMode=BMM_Additive
    translationSpace=BCS_BoneSpace
    alphaBoolBlend {blendInTime=0.5, blendOutTime=1}
    boneToModify {boneName=dyn_tail_r_MSTR}
    rotation {yaw=-15}
    translation {y=-5}
   }
   n19 = "Transform (Modify) Bone - Bone: base_cape_l"!(ComponentPose=@n18, bAlphaBoolEnabled:bool=True, Translation="0,0,1", Rotation="0,0,0", Scale="1,1,1") -> (Pose) {
    alphaInputType=Bool
    rotationMode=BMM_Additive
    rotationSpace=BCS_BoneSpace
    translationMode=BMM_Additive
    translationSpace=BCS_BoneSpace
    alphaBoolBlend {blendInTime=0.5, blendOutTime=1}
    boneToModify {boneName=base_cape_l}
    translation {z=1}
   }
   n20 = "Transform (Modify) Bone - Bone: base_cape_neck_l"!(ComponentPose=@n19, bAlphaBoolEnabled:bool=True, Translation="0,0,0", Rotation="0,0,-80", Scale="1,1,1") -> (Pose) {
    alphaInputType=Bool
    rotationMode=BMM_Additive
    rotationSpace=BCS_BoneSpace
    translationMode=BMM_Additive
    translationSpace=BCS_BoneSpace
    alphaBoolBlend {blendInTime=0.5, blendOutTime=0.5}
    boneToModify {boneName=base_cape_neck_l}
    rotation {roll=-80}
   }
   n21 = "Transform (Modify) Bone - Bone: base_cape_neck_r"!(ComponentPose=@n20, bAlphaBoolEnabled:bool=True, Translation="0,0,0", Rotation="0,0,60", Scale="1,1,1") -> (Pose) {
    alphaInputType=Bool
    rotationMode=BMM_Additive
    rotationSpace=BCS_BoneSpace
    translationMode=BMM_Additive
    translationSpace=BCS_BoneSpace
    alphaBoolBlend {blendInTime=0.5, blendOutTime=0.5}
    boneToModify {boneName=base_cape_neck_r}
    rotation {roll=60}
   }
   n22 = "Transform (Modify) Bone - Bone: dfrm_upperarm_twist_01_l"!(ComponentPose=@n21, bAlphaBoolEnabled:bool=True, Translation="0,0,0", Rotation="0,30,0", Scale="1,1,1") -> (Pose) {
    alphaInputType=Bool
    rotationMode=BMM_Additive
    rotationSpace=BCS_BoneSpace
    translationMode=BMM_Additive
    translationSpace=BCS_BoneSpace
    alphaBoolBlend {blendInTime=1, blendOutTime=1}
    boneToModify {boneName=dfrm_upperarm_twist_01_l}
    rotation {yaw=30}
   }
   Save cached pose 'PreDynamics_CACHE'(Pose=Component To Local(ComponentPose=@n22)) {cacheName=PreDynamics_CACHE}
   n23 = "Transform (Modify) Bone - Bone: base_cape_l"(
    ComponentPose=Local To Component(LocalPose=Use cached pose 'BaseLogic_CACHE'())
    bAlphaBoolEnabled=True
    Translation="0,0,6"
    Rotation="-35,-35,0"
    Scale="1,1,1"
   ) {
    alphaInputType=Bool
    rotationMode=BMM_Additive
    rotationSpace=BCS_BoneSpace
    translationMode=BMM_Additive
    translationSpace=BCS_BoneSpace
    alphaBoolBlend {blendInTime=0.2, blendOutTime=0.5}
    boneToModify {boneName=base_cape_l}
    rotation {pitch=-35, yaw=-35}
    translation {z=6}
   }
   n24 = RigidBody(ComponentPose=@n23, Alpha=0, OverrideWorldGravity="0,0,0") {
    bClampLinearTranslationLimitToRefPose=true
    bOverrideWorldGravity=true
    overridePhysicsAsset=&/F_MED_WiseBarn_Cape_Physics1
    simulationSpace=BaseBoneSpace
    alphaScaleBiasClamp {bClampResult=true, bInterpResult=true, interpSpeedIncreasing=1}
    baseBoneRef {boneName=spine_05}
    simSpaceSettings {worldAlpha=0.5}
   }
   n25 = "Transform (Modify) Bone - Bone: base_cape_l"(ComponentPose=@n24, bAlphaBoolEnabled=True, Translation="0,0,-1", Rotation="-10,0,0", Scale="1,1,1") {
    alphaInputType=Bool
    rotationMode=BMM_Additive
    rotationSpace=BCS_BoneSpace
    translationMode=BMM_Additive
    translationSpace=BCS_BoneSpace
    alphaBoolBlend {blendInTime=0.2, blendOutTime=0.5}
    boneToModify {boneName=base_cape_l}
    rotation {pitch=-10}
    translation {z=-1}
   }
   n26 = "Transform (Modify) Bone - Bone: base_cape_r"!(ComponentPose=@n25, bAlphaBoolEnabled:bool=True, Translation="0,0,-2", Rotation="0,0,0", Scale="1,1,1") -> (Pose) {
    alphaInputType=Bool
    rotationMode=BMM_Additive
    rotationSpace=BCS_BoneSpace
    translationMode=BMM_Additive
    translationSpace=BCS_BoneSpace
    alphaBoolBlend {blendInTime=0.2, blendOutTime=0.5}
    boneToModify {boneName=base_cape_r}
    translation {z=-2}
   }
   n27 = "Transform (Modify) Bone - Bone: base_cape_l"(ComponentPose=@n26, bAlphaBoolEnabled=True, Translation="0,0,1", Rotation="10,0,0", Scale="1,1,1") {
    alphaInputType=Bool
    rotationMode=BMM_Additive
    rotationSpace=BCS_BoneSpace
    translationMode=BMM_Additive
    translationSpace=BCS_BoneSpace
    alphaBoolBlend {blendInTime=0.2, blendOutTime=0.5}
    boneToModify {boneName=base_cape_l}
    rotation {pitch=10}
    translation {z=1}
   }
   n28 = "Transform (Modify) Bone - Bone: base_cape_r"(ComponentPose=@n27, bAlphaBoolEnabled=True, Translation="0,0,0", Rotation="10,0,0", Scale="1,1,1") {
    alphaInputType=Bool
    rotationMode=BMM_Additive
    rotationSpace=BCS_BoneSpace
    translationMode=BMM_Additive
    translationSpace=BCS_BoneSpace
    alphaBoolBlend {blendInTime=0.2, blendOutTime=0.5}
    boneToModify {boneName=base_cape_r}
    rotation {pitch=10}
   }
   n29 = Component To Local(ComponentPose=@n28)
   n30 = "Transform (Modify) Bone - Bone: base_cape_r"(
    ComponentPose=Local To Component(LocalPose=Use cached pose 'BaseLogic_CACHE'())
    bAlphaBoolEnabled=True
    Translation="0,0,0"
    Rotation="0,15,0"
    Scale="1,1,1"
   ) {
    alphaInputType=Bool
    rotationMode=BMM_Additive
    rotationSpace=BCS_BoneSpace
    translationMode=BMM_Additive
    translationSpace=BCS_BoneSpace
    alphaBoolBlend {blendInTime=0.2, blendOutTime=0.5}
    boneToModify {boneName=base_cape_r}
    rotation {yaw=15}
   }
   n31 = RigidBody(ComponentPose=@n30, Alpha=0, OverrideWorldGravity="0,0,0") {
    bClampLinearTranslationLimitToRefPose=true
    bOverrideWorldGravity=true
    overridePhysicsAsset=&/F_MED_WiseBarn_Cape_Backpack_Physics
    simulationSpace=BaseBoneSpace
    alphaScaleBiasClamp {bClampResult=true, bInterpResult=true, interpSpeedIncreasing=1}
    baseBoneRef {boneName=spine_05}
   }
   n32 = Component To Local(ComponentPose=@n31)
   n33 = Get isWearingBackpack()
   n34 = "Transform (Modify) Bone - Bone: base_cape_l"(
    ComponentPose=Local To Component(LocalPose=Blend Poses by bool(BlendPose_0=@n32, BlendPose_1=@n29, bActiveValue=@n33))
    bAlphaBoolEnabled=True
    Translation="0,0,9"
    Rotation="0,0,0"
    Scale="1,1,1"
   ) {
    alphaInputType=Bool
    rotationMode=BMM_Additive
    rotationSpace=BCS_BoneSpace
    translationMode=BMM_Additive
    translationSpace=BCS_BoneSpace
    alphaBoolBlend {blendInTime=0.2, blendOutTime=2}
    boneToModify {boneName=base_cape_l}
    translation {z=9}
   }
   n35 = "Transform (Modify) Bone - Bone: base_cape_mid"!(ComponentPose=@n34, bAlphaBoolEnabled:bool=True, Translation="0,0,9", Rotation="15,0,0", Scale="1,1,1") -> (Pose) {
    alphaInputType=Bool
    rotationMode=BMM_Additive
    rotationSpace=BCS_BoneSpace
    translationMode=BMM_Additive
    translationSpace=BCS_BoneSpace
    alphaBoolBlend {blendInTime=0.2, blendOutTime=2}
    boneToModify {boneName=base_cape_mid}
    rotation {pitch=15}
    translation {z=9}
   }
   n36 = "Transform (Modify) Bone - Bone: base_cape_r"(ComponentPose=@n35, bAlphaBoolEnabled=True, Translation="0,0,9", Rotation="20,0,0", Scale="1,1,1") {
    alphaInputType=Bool
    rotationMode=BMM_Additive
    rotationSpace=BCS_BoneSpace
    translationMode=BMM_Additive
    translationSpace=BCS_BoneSpace
    alphaBoolBlend {blendInTime=0.2, blendOutTime=2}
    boneToModify {boneName=base_cape_r}
    rotation {pitch=20}
    translation {z=9}
   }
   n37 = "Transform (Modify) Bone - Bone: dfrm_upperarm_twist_01_l"(ComponentPose=@n36, bAlphaBoolEnabled=True, Translation="0,0,0", Rotation="0,0,-25", Scale="1,1,1") {
    alphaInputType=Bool
    rotationMode=BMM_Additive
    rotationSpace=BCS_BoneSpace
    translationMode=BMM_Additive
    translationSpace=BCS_BoneSpace
    alphaBoolBlend {blendInTime=0.2, blendOutTime=0.5}
    boneToModify {boneName=dfrm_upperarm_twist_01_l}
    rotation {roll=-25}
   }
   n38 = "Transform (Modify) Bone - Bone: dfrm_lowerarm_twist_02_l"!(ComponentPose=@n37, bAlphaBoolEnabled:bool=True, Translation="0,0,0", Rotation="0,0,-45", Scale="1,1,1") -> (Pose) {
    alphaInputType=Bool
    rotationMode=BMM_Additive
    rotationSpace=BCS_BoneSpace
    translationMode=BMM_Additive
    translationSpace=BCS_BoneSpace
    alphaBoolBlend {blendInTime=0.2, blendOutTime=2}
    boneToModify {boneName=dfrm_lowerarm_twist_02_l}
    rotation {roll=-45}
   }
   n39 = "Transform (Modify) Bone - Bone: dfrm_upperarm_twist_01_r"!(ComponentPose=@n38, bAlphaBoolEnabled:bool=True, Translation="0,0,0", Rotation="0,0,-25", Scale="1,1,1") -> (Pose) {
    alphaInputType=Bool
    rotationMode=BMM_Additive
    rotationSpace=BCS_BoneSpace
    translationMode=BMM_Additive
    translationSpace=BCS_BoneSpace
    alphaBoolBlend {blendInTime=0.2, blendOutTime=0.5}
    boneToModify {boneName=dfrm_upperarm_twist_01_r}
    rotation {roll=-25}
   }
   n40 = "Transform (Modify) Bone - Bone: dfrm_lowerarm_twist_02_r"!(ComponentPose=@n39, bAlphaBoolEnabled:bool=True, Translation="0,0,0", Rotation="0,0,-45", Scale="1,1,1") -> (Pose) {
    alphaInputType=Bool
    rotationMode=BMM_Additive
    rotationSpace=BCS_BoneSpace
    translationMode=BMM_Additive
    translationSpace=BCS_BoneSpace
    alphaBoolBlend {blendInTime=0.2, blendOutTime=2}
    boneToModify {boneName=dfrm_lowerarm_twist_02_r}
    rotation {roll=-45}
   }
   n41 = "Transform (Modify) Bone - Bone: base_cape_r"(ComponentPose=@n40, bAlphaBoolEnabled=True, Translation="0,0,0", Rotation="-5,0,0", Scale="1,1,1") {
    alphaInputType=Bool
    rotationMode=BMM_Additive
    rotationSpace=BCS_BoneSpace
    translationMode=BMM_Additive
    translationSpace=BCS_BoneSpace
    alphaBoolBlend {blendInTime=1, blendOutTime=1}
    boneToModify {boneName=base_cape_r}
    rotation {pitch=-5}
   }
   n42 = "Transform (Modify) Bone - Bone: base_cape_mid"(ComponentPose=@n41, bAlphaBoolEnabled=True, Translation="0,0,0", Rotation="-10,0,0", Scale="1,1,1") {
    alphaInputType=Bool
    rotationMode=BMM_Additive
    rotationSpace=BCS_BoneSpace
    translationMode=BMM_Additive
    translationSpace=BCS_BoneSpace
    alphaBoolBlend {blendInTime=1, blendOutTime=1}
    boneToModify {boneName=base_cape_mid}
    rotation {pitch=-10}
   }
   n43 = "Transform (Modify) Bone - Bone: dfrm_upperarm_twist_01_r"(ComponentPose=@n42, bAlphaBoolEnabled=True, Translation="0,0,0", Rotation="0,0,-15", Scale="1,1,1") {
    alphaInputType=Bool
    rotationMode=BMM_Additive
    rotationSpace=BCS_BoneSpace
    translationMode=BMM_Additive
    translationSpace=BCS_BoneSpace
    alphaBoolBlend {blendInTime=0.3, blendOutTime=4}
    boneToModify {boneName=dfrm_upperarm_twist_01_r}
    rotation {roll=-15}
   }
   n44 = Anim Dynamics - Bone: dyn_wing_upper_r_02(ComponentPose=@n43, Alpha=0.3, ExternalForce="0,0,0") {
    angularDampingOverride=0.99
    angularSpringConstant=5000
    bAngularSpring=true
    bGravityOverrideInSimSpace=true
    bOverrideAngularDamping=true
    simulationSpace=RootRelative
    alphaBoolBlend {blendOption=Linear}
    boundBone {boneName=dyn_wing_upper_r_02}
    componentAppliedLinearAccClamp=<3500 3500 3500>
    componentLinearAccScale=<0.5 0.5 0.6>
    componentLinearVelScale=<0.5 0.5 0.6>
    gravityOverride {z=-980}
    physicsBodyDefinitions [{
      sphereCollisionRadius=0
      boundBone {boneName=dyn_wing_upper_r_02}
      boxExtents=<16 16 16>
      constraintSetup {angularLimitsMax=<1 25 10>, angularLimitsMin=<-1 -10 -25>, angularTarget=<1 0 0>}
      localJointOffset=<1 0 0>
     }]
    relativeSpaceBone {boneName=root}
   }
   n45 = Anim Dynamics - Bone: dyn_wing_upper_r_03(ComponentPose=@n44, Alpha=0.3, ExternalForce="0,0,0") {
    angularDampingOverride=0.99
    angularSpringConstant=5000
    bAngularSpring=true
    bGravityOverrideInSimSpace=true
    bOverrideAngularDamping=true
    simulationSpace=RootRelative
    alphaBoolBlend {blendOption=Linear}
    boundBone {boneName=dyn_wing_upper_r_03}
    componentAppliedLinearAccClamp=<3500 3500 3500>
    componentLinearAccScale=<0.5 0.5 0.6>
    componentLinearVelScale=<0.5 0.5 0.6>
    gravityOverride {z=-980}
    physicsBodyDefinitions [{
      sphereCollisionRadius=0
      boundBone {boneName=dyn_wing_upper_r_03}
      boxExtents=<16 16 16>
      constraintSetup {angularLimitsMax=<1 25 10>, angularLimitsMin=<-1 -10 -25>, angularTarget=<1 0 0>}
      localJointOffset=<1 0 0>
     }]
    relativeSpaceBone {boneName=root}
   }
   n46 = Anim Dynamics - Bone: dyn_wing_upper_r_01(ComponentPose=@n45, Alpha=0.3, ExternalForce="0,0,0") {
    angularDampingOverride=0.99
    angularSpringConstant=5000
    bAngularSpring=true
    bGravityOverrideInSimSpace=true
    bOverrideAngularDamping=true
    simulationSpace=RootRelative
    alphaBoolBlend {blendOption=Linear}
    boundBone {boneName=dyn_wing_upper_r_01}
    componentAppliedLinearAccClamp=<3500 3500 3500>
    componentLinearAccScale=<0.5 0.5 0.6>
    componentLinearVelScale=<0.5 0.5 0.6>
    gravityOverride {z=-980}
    physicsBodyDefinitions [{
      sphereCollisionRadius=0
      boundBone {boneName=dyn_wing_upper_r_01}
      boxExtents=<16 16 16>
      constraintSetup {angularLimitsMax=<1 25 10>, angularLimitsMin=<-1 -5 -25>, angularTarget=<1 0 0>}
      localJointOffset=<1 0 0>
     }]
    relativeSpaceBone {boneName=root}
   }
   n47 = Anim Dynamics - Bone: dyn_wing_upper_r_04(ComponentPose=@n46, Alpha=0.3, ExternalForce="0,0,0") {
    angularDampingOverride=0.99
    angularSpringConstant=5000
    bAngularSpring=true
    bGravityOverrideInSimSpace=true
    bOverrideAngularDamping=true
    simulationSpace=RootRelative
    alphaBoolBlend {blendOption=Linear}
    boundBone {boneName=dyn_wing_upper_r_04}
    componentAppliedLinearAccClamp=<3500 3500 3500>
    componentLinearAccScale=<0.5 0.5 0.6>
    componentLinearVelScale=<0.5 0.5 0.6>
    gravityOverride {z=-980}
    physicsBodyDefinitions [{
      sphereCollisionRadius=0
      boundBone {boneName=dyn_wing_upper_r_04}
      boxExtents=<16 16 16>
      constraintSetup {angularLimitsMax=<1 25 10>, angularLimitsMin=<-1 0 -25>, angularTarget=<1 0 0>}
      localJointOffset=<1 0 0>
     }]
    relativeSpaceBone {boneName=root}
   }
   n48 = Anim Dynamics - Bone: dyn_wing_upper_r_05(ComponentPose=@n47, Alpha=0.3, ExternalForce="0,0,0") {
    angularDampingOverride=0.99
    angularSpringConstant=5000
    bAngularSpring=true
    bGravityOverrideInSimSpace=true
    bOverrideAngularDamping=true
    simulationSpace=RootRelative
    alphaBoolBlend {blendOption=Linear}
    boundBone {boneName=dyn_wing_upper_r_05}
    componentAppliedLinearAccClamp=<3500 3500 3500>
    componentLinearAccScale=<0.5 0.5 0.6>
    componentLinearVelScale=<0.5 0.5 0.6>
    gravityOverride {z=-980}
    physicsBodyDefinitions [{
      sphereCollisionRadius=0
      boundBone {boneName=dyn_wing_upper_r_05}
      boxExtents=<16 16 16>
      constraintSetup {angularLimitsMax=<1 25 10>, angularLimitsMin=<-1 0 -25>, angularTarget=<1 0 0>}
      localJointOffset=<1 0 0>
     }]
    relativeSpaceBone {boneName=root}
   }
   n49 = Anim Dynamics - Bone: dyn_wing_upper_l_02(ComponentPose=@n48, Alpha=0.3, ExternalForce="0,0,0") {
    angularDampingOverride=0.99
    angularSpringConstant=5000
    bAngularSpring=true
    bGravityOverrideInSimSpace=true
    bOverrideAngularDamping=true
    simulationSpace=RootRelative
    alphaBoolBlend {blendOption=Linear}
    boundBone {boneName=dyn_wing_upper_l_02}
    componentAppliedLinearAccClamp=<3500 3500 3500>
    componentLinearAccScale=<0.5 0.5 0.6>
    componentLinearVelScale=<0.5 0.5 0.6>
    gravityOverride {z=-980}
    physicsBodyDefinitions [{
      sphereCollisionRadius=0
      boundBone {boneName=dyn_wing_upper_l_02}
      boxExtents=<16 16 16>
      constraintSetup {angularLimitsMax=<1 25 10>, angularLimitsMin=<-1 -10 -25>, angularTarget=<1 0 0>}
      localJointOffset=<-1 0 0>
     }]
    relativeSpaceBone {boneName=root}
   }
   n50 = Anim Dynamics - Bone: dyn_wing_upper_l_03(ComponentPose=@n49, Alpha=0.3, ExternalForce="0,0,0") {
    angularDampingOverride=0.99
    angularSpringConstant=5000
    bAngularSpring=true
    bGravityOverrideInSimSpace=true
    bOverrideAngularDamping=true
    simulationSpace=RootRelative
    alphaBoolBlend {blendOption=Linear}
    boundBone {boneName=dyn_wing_upper_l_03}
    componentAppliedLinearAccClamp=<3500 3500 3500>
    componentLinearAccScale=<0.5 0.5 0.6>
    componentLinearVelScale=<0.5 0.5 0.6>
    gravityOverride {z=-980}
    physicsBodyDefinitions [{
      sphereCollisionRadius=0
      boundBone {boneName=dyn_wing_upper_l_03}
      boxExtents=<16 16 16>
      constraintSetup {angularLimitsMax=<1 25 10>, angularLimitsMin=<-1 -10 -25>, angularTarget=<1 0 0>}
      localJointOffset=<-1 0 0>
     }]
    relativeSpaceBone {boneName=root}
   }
   n51 = Anim Dynamics - Bone: dyn_wing_upper_l_01(ComponentPose=@n50, Alpha=0.3, ExternalForce="0,0,0") {
    angularDampingOverride=0.99
    angularSpringConstant=5000
    bAngularSpring=true
    bGravityOverrideInSimSpace=true
    bOverrideAngularDamping=true
    simulationSpace=RootRelative
    alphaBoolBlend {blendOption=Linear}
    boundBone {boneName=dyn_wing_upper_l_01}
    componentAppliedLinearAccClamp=<3500 3500 3500>
    componentLinearAccScale=<0.5 0.5 0.6>
    componentLinearVelScale=<0.5 0.5 0.6>
    gravityOverride {z=-980}
    physicsBodyDefinitions [{
      sphereCollisionRadius=0
      boundBone {boneName=dyn_wing_upper_l_01}
      boxExtents=<16 16 16>
      constraintSetup {angularLimitsMax=<1 25 10>, angularLimitsMin=<-1 -5 -25>, angularTarget=<1 0 0>}
      localJointOffset=<-1 0 0>
     }]
    relativeSpaceBone {boneName=root}
   }
   n52 = Anim Dynamics - Bone: dyn_wing_upper_l_04(ComponentPose=@n51, Alpha=0.3, ExternalForce="0,0,0") {
    angularDampingOverride=0.99
    angularSpringConstant=5000
    bAngularSpring=true
    bGravityOverrideInSimSpace=true
    bOverrideAngularDamping=true
    simulationSpace=RootRelative
    alphaBoolBlend {blendOption=Linear}
    boundBone {boneName=dyn_wing_upper_l_04}
    componentAppliedLinearAccClamp=<3500 3500 3500>
    componentLinearAccScale=<0.5 0.5 0.6>
    componentLinearVelScale=<0.5 0.5 0.6>
    gravityOverride {z=-980}
    physicsBodyDefinitions [{
      sphereCollisionRadius=0
      boundBone {boneName=dyn_wing_upper_l_04}
      boxExtents=<16 16 16>
      constraintSetup {angularLimitsMax=<1 25 10>, angularLimitsMin=<-1 0 -25>, angularTarget=<1 0 0>}
      localJointOffset=<-1 0 0>
     }]
    relativeSpaceBone {boneName=root}
   }
   n53 = Anim Dynamics - Bone: dyn_wing_upper_l_05(ComponentPose=@n52, Alpha=0.3, ExternalForce="0,0,0") {
    angularDampingOverride=0.99
    angularSpringConstant=5000
    bAngularSpring=true
    bGravityOverrideInSimSpace=true
    bOverrideAngularDamping=true
    simulationSpace=RootRelative
    alphaBoolBlend {blendOption=Linear}
    boundBone {boneName=dyn_wing_upper_l_05}
    componentAppliedLinearAccClamp=<3500 3500 3500>
    componentLinearAccScale=<0.5 0.5 0.6>
    componentLinearVelScale=<0.5 0.5 0.6>
    gravityOverride {z=-980}
    physicsBodyDefinitions [{
      sphereCollisionRadius=0
      boundBone {boneName=dyn_wing_upper_l_05}
      boxExtents=<16 16 16>
      constraintSetup {angularLimitsMax=<1 25 10>, angularLimitsMin=<-1 0 -25>, angularTarget=<1 0 0>}
      localJointOffset=<-1 0 0>
     }]
    relativeSpaceBone {boneName=root}
   }
   n54 = Anim Dynamics - Bone: dyn_calf_feather_02(ComponentPose=@n53, Alpha=0.5, ExternalForce="0,0,0") {
    angularDampingOverride=0.9999
    angularSpringConstant=75
    bAngularSpring=true
    bGravityOverrideInSimSpace=true
    bOverrideAngularDamping=true
    gravityScale=0.8
    simulationSpace=BoneRelative
    alphaBoolBlend {blendOption=Linear}
    boundBone {boneName=dyn_calf_feather_02}
    componentAppliedLinearAccClamp=<3500 3500 3500>
    componentLinearAccScale=<1.5 1.5 1.25>
    componentLinearVelScale=<1.5 1.5 1.25>
    gravityOverride {z=-980}
    physicsBodyDefinitions [{
      sphereCollisionRadius=0
      boundBone {boneName=dyn_calf_feather_02}
      boxExtents=<17 17 10>
      constraintSetup {angularLimitsMax=<5 3 15>, angularLimitsMin=<-5 -15 -15>, angularTarget=<1 0 0>}
      localJointOffset=<4 0.5 0>
     }]
    relativeSpaceBone {boneName=root}
   }
   n55 = Anim Dynamics - Bone: dyn_calf_feather_04(ComponentPose=@n54, Alpha=0.5, ExternalForce="0,0,0") {
    angularDampingOverride=0.9999
    angularSpringConstant=75
    bAngularSpring=true
    bGravityOverrideInSimSpace=true
    bOverrideAngularDamping=true
    gravityScale=0.8
    simulationSpace=BoneRelative
    alphaBoolBlend {blendOption=Linear}
    boundBone {boneName=dyn_calf_feather_04}
    componentAppliedLinearAccClamp=<3500 3500 3500>
    componentLinearAccScale=<1.5 1.5 1.25>
    componentLinearVelScale=<1.5 1.5 1.25>
    gravityOverride {z=-980}
    physicsBodyDefinitions [{
      sphereCollisionRadius=0
      boundBone {boneName=dyn_calf_feather_04}
      boxExtents=<17 17 10>
      constraintSetup {angularLimitsMax=<5 3 15>, angularLimitsMin=<-5 -15 -15>, angularTarget=<1 0 0>}
      localJointOffset=<4 0 0>
     }]
    relativeSpaceBone {boneName=root}
   }
   n56 = Anim Dynamics - Bone: dyn_calf_feather_05(ComponentPose=@n55, Alpha=0.5, ExternalForce="0,0,0") {
    angularDampingOverride=0.9999
    angularSpringConstant=75
    bAngularSpring=true
    bGravityOverrideInSimSpace=true
    bOverrideAngularDamping=true
    gravityScale=0.8
    simulationSpace=BoneRelative
    alphaBoolBlend {blendOption=Linear}
    boundBone {boneName=dyn_calf_feather_05}
    componentAppliedLinearAccClamp=<3500 3500 3500>
    componentLinearAccScale=<1.5 1.5 1.25>
    componentLinearVelScale=<1.5 1.5 1.25>
    gravityOverride {z=-980}
    physicsBodyDefinitions [{
      sphereCollisionRadius=0
      boundBone {boneName=dyn_calf_feather_05}
      boxExtents=<17 17 10>
      constraintSetup {angularLimitsMax=<5 3 15>, angularLimitsMin=<-5 -15 -15>, angularTarget=<1 0 0>}
      localJointOffset=<3 0 0>
     }]
    relativeSpaceBone {boneName=root}
   }
   n57 = Anim Dynamics - Bone: dyn_calf_feather_08(ComponentPose=@n56, Alpha=0.5, ExternalForce="0,0,0") {
    angularDampingOverride=0.9999
    angularSpringConstant=75
    bAngularSpring=true
    bGravityOverrideInSimSpace=true
    bOverrideAngularDamping=true
    gravityScale=0.8
    simulationSpace=BoneRelative
    alphaBoolBlend {blendOption=Linear}
    boundBone {boneName=dyn_calf_feather_08}
    componentAppliedLinearAccClamp=<3500 3500 3500>
    componentLinearAccScale=<1.5 1.5 1.25>
    componentLinearVelScale=<1.5 1.5 1.25>
    gravityOverride {z=-980}
    physicsBodyDefinitions [{
      sphereCollisionRadius=0
      boundBone {boneName=dyn_calf_feather_08}
      boxExtents=<17 17 10>
      constraintSetup {angularLimitsMax=<5 3 15>, angularLimitsMin=<-5 -15 -15>, angularTarget=<1 0 0>}
      localJointOffset=<4 0 0>
     }]
    relativeSpaceBone {boneName=root}
   }
   n58 = Anim Dynamics - Bone: dyn_calf_feather_l_01(ComponentPose=@n57, Alpha=0.5, ExternalForce="0,0,0") {
    angularDampingOverride=0.9999
    angularSpringConstant=75
    bAngularSpring=true
    bGravityOverrideInSimSpace=true
    bOverrideAngularDamping=true
    gravityScale=0.8
    simulationSpace=BoneRelative
    alphaBoolBlend {blendOption=Linear}
    boundBone {boneName=dyn_calf_feather_l_01}
    componentAppliedLinearAccClamp=<3500 3500 3500>
    componentLinearAccScale=<1.5 1.5 1.25>
    componentLinearVelScale=<1.5 1.5 1.25>
    gravityOverride {z=-980}
    physicsBodyDefinitions [{
      sphereCollisionRadius=0
      boundBone {boneName=dyn_calf_feather_l_01}
      boxExtents=<17 17 10>
      constraintSetup {angularLimitsMax=<5 3 15>, angularLimitsMin=<-5 -15 -15>, angularTarget=<1 0 0>}
      localJointOffset=<3 -0.5 0>
     }]
    relativeSpaceBone {boneName=root}
   }
   n59 = Anim Dynamics - Bone: dyn_calf_feather_l_04(ComponentPose=@n58, Alpha=0.5, ExternalForce="0,0,0") {
    angularDampingOverride=0.9999
    angularSpringConstant=75
    bAngularSpring=true
    bGravityOverrideInSimSpace=true
    bOverrideAngularDamping=true
    gravityScale=0.8
    simulationSpace=BoneRelative
    alphaBoolBlend {blendOption=Linear}
    boundBone {boneName=dyn_calf_feather_l_04}
    componentAppliedLinearAccClamp=<3500 3500 3500>
    componentLinearAccScale=<1.5 1.5 1.25>
    componentLinearVelScale=<1.5 1.5 1.25>
    gravityOverride {z=-980}
    physicsBodyDefinitions [{
      sphereCollisionRadius=0
      boundBone {boneName=dyn_calf_feather_l_04}
      boxExtents=<17 17 10>
      constraintSetup {angularLimitsMax=<5 3 15>, angularLimitsMin=<-5 -15 -15>, angularTarget=<1 0 0>}
      localJointOffset=<4 0 0>
     }]
    relativeSpaceBone {boneName=root}
   }
   n60 = Anim Dynamics - Bone: dyn_calf_feather_l_06(ComponentPose=@n59, Alpha=0.5, ExternalForce="0,0,0") {
    angularDampingOverride=0.9999
    angularSpringConstant=75
    bAngularSpring=true
    bGravityOverrideInSimSpace=true
    bOverrideAngularDamping=true
    gravityScale=0.8
    simulationSpace=BoneRelative
    alphaBoolBlend {blendOption=Linear}
    boundBone {boneName=dyn_calf_feather_l_06}
    componentAppliedLinearAccClamp=<3500 3500 3500>
    componentLinearAccScale=<1.5 1.5 1.25>
    componentLinearVelScale=<1.5 1.5 1.25>
    gravityOverride {z=-980}
    physicsBodyDefinitions [{
      sphereCollisionRadius=0
      boundBone {boneName=dyn_calf_feather_l_06}
      boxExtents=<17 17 10>
      constraintSetup {angularLimitsMax=<5 3 15>, angularLimitsMin=<-5 -15 -15>, angularTarget=<1 0 0>}
      localJointOffset=<4 0.5 0>
     }]
    relativeSpaceBone {boneName=root}
   }
   n61 = Anim Dynamics - Bone: dyn_calf_feather_l_08(ComponentPose=@n60, Alpha=0.5, ExternalForce="0,0,0") {
    angularDampingOverride=0.9999
    angularSpringConstant=75
    bAngularSpring=true
    bGravityOverrideInSimSpace=true
    bOverrideAngularDamping=true
    gravityScale=0.8
    simulationSpace=BoneRelative
    alphaBoolBlend {blendOption=Linear}
    boundBone {boneName=dyn_calf_feather_l_08}
    componentAppliedLinearAccClamp=<3500 3500 3500>
    componentLinearAccScale=<1.5 1.5 1.25>
    componentLinearVelScale=<1.5 1.5 1.25>
    gravityOverride {z=-980}
    physicsBodyDefinitions [{
      sphereCollisionRadius=0
      boundBone {boneName=dyn_calf_feather_l_08}
      boxExtents=<17 17 10>
      constraintSetup {angularLimitsMax=<5 3 15>, angularLimitsMin=<-5 -15 -15>, angularTarget=<1 0 0>}
      localJointOffset=<4 0 0>
     }]
    relativeSpaceBone {boneName=root}
   }
   n62 = Anim Dynamics - Bone: dyn_calf_feather_l_10(ComponentPose=@n61, Alpha=0.5, ExternalForce="0,0,0") {
    angularDampingOverride=0.9999
    angularSpringConstant=75
    bAngularSpring=true
    bGravityOverrideInSimSpace=true
    bOverrideAngularDamping=true
    gravityScale=0.8
    simulationSpace=BoneRelative
    alphaBoolBlend {blendOption=Linear}
    boundBone {boneName=dyn_calf_feather_l_10}
    componentAppliedLinearAccClamp=<3500 3500 3500>
    componentLinearAccScale=<1.5 1.5 1.25>
    componentLinearVelScale=<1.5 1.5 1.25>
    gravityOverride {z=-980}
    physicsBodyDefinitions [{
      sphereCollisionRadius=0
      boundBone {boneName=dyn_calf_feather_l_10}
      boxExtents=<17 17 10>
      constraintSetup {angularLimitsMax=<5 3 15>, angularLimitsMin=<-5 -15 -15>, angularTarget=<1 0 0>}
      localJointOffset=<4 0 0>
     }]
    relativeSpaceBone {boneName=root}
   }
   Save cached pose 'Dynamics_CACHE'(Pose=Component To Local(ComponentPose=@n62)) {cacheName=Dynamics_CACHE}
  ]}
 EventGraph {nodes [
   Event BlueprintUpdateAnimation()
   > n1 = Sequence!(execute) -> (then_0, then_1, then_2:Exec, then_3:Exec, then_4:Exec, then_5:Exec, then_6:Exec, then_7:Exec, then_8:Exec, then_9:Exec)
   n2 = Get bRigidBodyEnabled()
   Set RigidBodyOffAndCrouching(execute=@n1.then_2, RigidBodyOffAndCrouching:bool=AND Boolean(A=NOT Boolean(A=@n2), B=Get bIsCrouching())) -> (Output_Get:bool)
   n3 = Get bIsSkydiveDiveMode()
   n4 = Get isCustomEmotePlaying() -> (isCustomEmotePlaying:bool)
   n5 = Branch(execute=@n1.then_6, Condition=OR Boolean(A=@n3, B=@n4))
   > n6 = Branch(Condition=@n3)
   > n7 = Branch(Condition=@n3)
   > Delay(Duration=0.3)
   > Set SkydiveDownAlpha(SkydiveDownAlpha:double=0) -> (Output_Get:double)
   Delay(execute=@n7.else, Duration=0.6)
   > Set SkydiveDownAlpha(SkydiveDownAlpha=1)
   n8 = Branch(execute=@n6.else, Condition=@n4)
   > Set SkydiveDownAlpha(SkydiveDownAlpha=1)
   Set SkydiveDownAlpha(execute=@n8.else, SkydiveDownAlpha=1)
   Delay(execute=@n5.else, Duration=0.3)
   > Set SkydiveDownAlpha(SkydiveDownAlpha=1)
   n9 = Get Skel Mesh()
   n10 = AND Boolean(
    A="float > float"(
     A=GetAlphaFromReferenceTransformDelta(SkeletalMeshComponent=@n9, BoneName=spine_01, BaseName=root, DeltaComponentType=EulerX, InputRange=(X=0,Y=-55), TwistAxis="1,0,0")
     B=0.4
    )
    B=NOT Boolean(A=Get bIsMoving2D())
   )
   n11 = Set RotWithoutMove_R(execute=@n1.then_3, RotWithoutMove_R:bool=@n10) -> (Output_Get:bool)
   > Set RotRWithoutMoveAndPickaxeEquiped(RotRWithoutMoveAndPickaxeEquiped:bool=AND Boolean(A=@n11, B=Get PickAxe Equipped() -> (PickAxe Equipped:bool))) -> (Output_Get:bool)
   n12 = GetAnimInstance(self=GetOwnerSkeletalMeshComponent())
   > Is Valid(InputObject=@n12)
   > Set isCustomEmotePlaying(isCustomEmotePlaying:bool=Montage_IsPlaying(self=@n12, Montage=/Game/Animation/Game/MainPlayer/Emotes/Emote_Breakdance_Female_M)) -> (Output_Get:bool)
   n13 = "Gravity Override State Function (Parameter Asset) (Nativized)"(
    execute=@n1.then_0
    CharacterPartAnimInstance=Self-Reference()
    GameTimeInSeconds=GetGameTimeInSeconds()
    Parameters=&/F_MED_Wisebarn_Parameters
   )
   > Set Gravity Override(Gravity Override=@n13)
   n14 = Get bIsSprinting()
   n15 = Get bIsPlayingEmote()
   > Set sprint(sprint:double="To Float (Boolean)"(InBool=OR Boolean(A=@n14, B=@n15))) -> (Output_Get:double)
   > Set emote(emote:double="To Float (Boolean)"(InBool=Get bIsPlayingEmote())) -> (Output_Get:double)
   > Set isSkyDivingFloat(isSkyDivingFloat:double="To Float (Boolean)"(InBool=Get bIsSkyDiving())) -> (Output_Get:double)
   > n16 = Cast To FortPlayerAnimInstance(Object=GetAnimInstance(self=GetOwnerSkeletalMeshComponent()))
   > Set pistol(pistol:double=-0.6) -> (Output_Get:double)
   > n17 = Switch on EFortWeaponCoreAnimation!(execute, Selection:EFortWeaponCoreAnimation Enum=Get WeaponCoreAnim(self=@n16)) -> (
    Melee:Exec
    Pistol:Exec
    Shotgun:Exec
    PaperBlueprint:Exec
    Rifle:Exec
    MeleeOneHand:Exec
    MachinePistol:Exec
    RocketLauncher:Exec
    GrenadeLauncher:Exec
    GoingCommando:Exec
    AssaultRifle:Exec
    TacticalShotgun:Exec
    SniperRifle:Exec
    TrapPlacement:Exec
    ShoulderLauncher:Exec
    AbilityDecoTool:Exec
    Crossbow:Exec
    C4:Exec
    RemoteControl:Exec
    DualWield:Exec
    AR_BullPup:Exec
    AR_ForwardGrip:Exec
    MedPackPaddles:Exec
    SMG_P90:Exec
    AR_DrumGun:Exec
    Consumable_Small:Exec
    Consumable_Large:Exec
    Balloon:Exec
    MountedTurret:Exec
    CreativeTool:Exec
    ExplosiveBow:Exec
    AshtonIndigo:Exec
    AshtonChicago:Exec
    MeleeDualWield:Exec
    Unarmed:Exec
   )
   Set pistol(execute=@n17.Pistol, pistol=-0.1)
   n18 = Get bIsSkyDiving()
   n19 = Get bIsPlayingEmote()
   n20 = Branch(execute=@n1.then_1, Condition=AND Boolean!(A=Get bIsInFrontEnd(), B=NOT Boolean(A=@n19), C:bool=NOT Boolean(A=@n18)) -> (ReturnValue))
   > Set frontend(frontend:double=1) -> (Output_Get:double)
   Set frontend(execute=@n20.else, frontend=0)
   n21 = Get bRigidBodyEnabled()
   Set ASWonFE(execute=@n1.then_7, ASWonFE:bool=AND Boolean!(A=Get bIsInFrontEnd(), B=NOT Boolean(A=@n21)) -> (ReturnValue)) -> (Output_Get:bool)
   Set AirAdjust(execute=@n1.then_5, AirAdjust:bool=OR Boolean(A=Get bIsSkydiveDiveMode(), B=Get EnterSkydiving() -> (EnterSkydiving:bool))) -> (Output_Get:bool)
   n22 = Cast To FortPlayerAnimInstance(execute=@n1.then_4, Object=GetAnimInstance(self=GetOwnerSkeletalMeshComponent()))
   > Set PickAxe Equipped(PickAxe Equipped:bool=false) -> (Output_Get:bool)
   > Switch on EFortWeaponCoreAnimation(Selection=Get WeaponCoreAnim(self=@n22))
   > n23 = Branch(Condition=NOT Boolean(A=Get bIsMoving2D()))
   > Set PickAxe Equipped(PickAxe Equipped=true)
   Set PickAxe Equipped(execute=@n23.else, PickAxe Equipped=false)
   Event BlueprintInitializeAnimation()
   > Set Skel Mesh(Skel Mesh=GetOwnerSkeletalMeshComponent())
   Event OnEndPlayingEmote() {is=title K2Node_Event}
   > n24 = Branch()
   Do Once(execute=@n24.else, Reset=@n24.then, Start Closed=true)
   > ResetDynamics(InTeleportType=TeleportPhysics)
   > ResetDynamics(InTeleportType=ResetPhysics)
   Event OnCharacterCustomizationLoaded()
   n25 = CheckForBackbling()
   > Set isWearingBackpack(isWearingBackpack:bool=AND Boolean!(A=@n25, B=NOT Boolean(A=@n25.bIsBackblingACape), C:bool=NOT Boolean(A=@n25.bIsBackblingLong)) -> (ReturnValue)) -> (Output_Get:bool)
   Event OnWeaponChanged()
   > Set WeaponChanged(WeaponChanged:bool=true) -> (Output_Get:bool)
   > Delay(Duration=0.2)
   > Set WeaponChanged(WeaponChanged=false)
   Event OnBeginSkydiving()
   > Set EnterSkydiving(EnterSkydiving:bool=true) -> (Output_Get:bool)
   > Delay(Duration=0.6)
   > Set EnterSkydiving(EnterSkydiving=false)
   Montage_IsPlaying(self=@n12, Montage=/BRCosmetics/Animation/Game/MainPlayer/Emotes/BalletSpin/Emote_BalletSpin_CMF_M)
   Delay(Duration=0.2)
   n26 = AND Boolean!(A=@n10, B=Get bIsTargeting()) -> (ReturnValue)
   OR Boolean(A=@n26, B=Get bIsPlayingEmote())
   Set RotWithoutMoveRAndTargetingOrEmote(RotWithoutMoveRAndTargetingOrEmote:bool=@n26) -> (Output_Get:bool)
  ]}
]
variables [
 AirAdjust {type=boolean}
 ASWonFE {type=boolean}
 emote {type=number}
 EnterSkydiving {type=boolean}
 frontend {type=number}
 gauntless {type=boolean}
 Gravity Override {type=Vector}
 isCustomEmotePlaying {type=boolean}
 isSkyDivingFloat {type=number}
 isWearingBackpack {type=boolean}
 PickAxe Equipped {type=boolean}
 pistol {type=number}
 RigidBodyOffAndCrouching {type=boolean}
 RotRWithoutMoveAndPickaxeEquiped {type=boolean}
 RotWithoutMove_R {type=boolean}
 RotWithoutMoveRAndTargetingOrEmote {type=boolean}
 Skel Mesh {type=SkeletalMeshComponent}
 SkydiveDownAlpha {type=number}
 sprint {type=number}
 WeaponChanged {type=boolean}
]
