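# Physics asset for the WiseBarn cape on the medium female player body, with capsule bodies for cape, spine, backpack and upper wing bones.
# Three cape chains (left, middle, right) hang from base bodies on spine_02, each segment having tight twist and narrow swing limits.
# Angularly free constraints cross-link the side chains to the middle at each level, and also tie the backpack and wing capsules to spine_02.
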PhysicsAsset /BRCosmetics/Characters/Player/Female/Medium/Bodies/M_MED_WiseBarn/Meshes/F_MED_WiseBarn_Cape_UsingBP_Physics
bodies {
 attach_backpack [{shapeType=Capsule, length=66, radius=22, shapeName=attach_backpack_capsule, center=<0 -48 -9>, rotation=<rot 8 -90 90>}]
 base_cape_l [{shapeType=Capsule, length=0.505, radius=0.505, shapeName=base_cape_l_capsule, rotation=<rot 0 -90 0>}]
 base_cape_mid [{shapeType=Capsule, length=0.505, radius=0.505, shapeName=base_cape_mid_capsule, rotation=<rot 0 -90 0>}]
 base_cape_r [{shapeType=Capsule, length=0.505, radius=0.505, shapeName=base_cape_r_capsule, rotation=<rot 0 -90 0>}]
 dyn_cape_l_01 [{shapeType=Capsule, length=16, radius=9, shapeName=dyn_cape_l_01_capsule, center=<11 0 10>, rotation=<rot 80 0 0>}]
 dyn_cape_l_02 [{shapeType=Capsule, length=16, radius=9, shapeName=dyn_cape_l_02_capsule, center=<11 -2.5 6>, rotation=<rot 85 0 0>}]
 dyn_cape_l_03 [{shapeType=Capsule, length=16, radius=9, shapeName=dyn_cape_l_03_capsule, center=<11 -2.5 6>, rotation=<rot 90 0 0>}]
 dyn_cape_mid_01 [{shapeType=Capsule, length=16, radius=9, shapeName=dyn_cape_mid_01_capsule, center=<11 0 8>, rotation=<rot 90 0 0>}]
 dyn_cape_mid_02 [{shapeType=Capsule, length=16, radius=9, shapeName=dyn_cape_mid_02_capsule, center=<11 0 6>, rotation=<rot 85 0 0>}]
 dyn_cape_mid_03 [{shapeType=Capsule, length=16, radius=9, shapeName=dyn_cape_mid_03_capsule, center=<11 0 6>, rotation=<rot 90 0 0>}]
 dyn_cape_r_01 [{shapeType=Capsule, length=16, radius=9, shapeName=dyn_cape_r_01_capsule, center=<11 0 10>, rotation=<rot 80 0 0>}]
 dyn_cape_r_02 [{shapeType=Capsule, length=16, radius=9, shapeName=dyn_cape_r_02_capsule, center=<11 2.5 6>, rotation=<rot 85 0 0>}]
 dyn_cape_r_03 [{shapeType=Capsule, length=16, radius=9, shapeName=dyn_cape_r_03_capsule, center=<11 2.5 6>, rotation=<rot 90 0 0>}]
 dyn_wing_upper_l_03 [{shapeType=Capsule, length=22, radius=4, shapeName=dyn_wing_upper_l_03_capsule, center=<-9 -5 -2>, rotation=<rot 90 15 0>}]
 dyn_wing_upper_l_04 [{shapeType=Capsule, length=22, radius=4, shapeName=dyn_wing_upper_l_04_capsule, center=<-14 5 0>, rotation=<rot 0 110 90>}]
 dyn_wing_upper_r_03 [{shapeType=Capsule, length=22, radius=4, shapeName=dyn_wing_upper_r_03_capsule, center=<9 5 2>, rotation=<rot 90 15 0>}]
 dyn_wing_upper_r_04 [{shapeType=Capsule, length=22, radius=5, shapeName=dyn_wing_upper_r_04_capsule, center=<14 -4 0>, rotation=<rot 0 110 90>}]
 spine_01 [{shapeType=Capsule, length=41.1229, radius=8.3598, shapeName=spine_01_capsule, center=<-6.2 2.9403 0>}]
 spine_02 [{shapeType=Capsule, length=12.3676, radius=14.9416, shapeName=spine_02_capsule, center=<-0.6554 3.8926 0>}]
}
constraints [
 {bone1Name=dyn_cape_l_01, bone2Name=base_cape_l, swing1LimitDegrees=3, swing2LimitDegrees=20, twistLimitDegrees=1}
 {bone1Name=dyn_cape_l_02, bone2Name=dyn_cape_l_01, swing1LimitDegrees=5, swing2LimitDegrees=20, twistLimitDegrees=1}
 {bone1Name=dyn_cape_l_03, bone2Name=dyn_cape_l_02, swing1LimitDegrees=10, swing2LimitDegrees=20, twistLimitDegrees=1}
 {bone1Name=dyn_cape_mid_01, bone2Name=base_cape_mid, swing1LimitDegrees=3, swing2LimitDegrees=20, twistLimitDegrees=1}
 {bone1Name=dyn_cape_mid_02, bone2Name=dyn_cape_mid_01, swing1LimitDegrees=3, swing2LimitDegrees=20, twistLimitDegrees=1}
 {bone1Name=dyn_cape_mid_03, bone2Name=dyn_cape_mid_02, swing1LimitDegrees=3, swing2LimitDegrees=20, twistLimitDegrees=1}
 {bone1Name=dyn_cape_r_01, bone2Name=base_cape_r, swing1LimitDegrees=3, swing2LimitDegrees=20, twistLimitDegrees=1}
 {bone1Name=dyn_cape_r_02, bone2Name=dyn_cape_r_01, swing1LimitDegrees=5, swing2LimitDegrees=20, twistLimitDegrees=1}
 {bone1Name=dyn_cape_r_03, bone2Name=dyn_cape_r_02, swing1LimitDegrees=10, swing2LimitDegrees=20, twistLimitDegrees=1}
 {
  bone1Name=base_cape_l
  bone2Name=spine_02
  swing1LimitDegrees=45
  swing1Motion=Free
  swing2LimitDegrees=45
  swing2Motion=Free
  twistLimitDegrees=45
  twistMotion=Free
 }
 {
  bone1Name=base_cape_r
  bone2Name=spine_02
  swing1LimitDegrees=45
  swing1Motion=Free
  swing2LimitDegrees=45
  swing2Motion=Free
  twistLimitDegrees=45
  twistMotion=Free
 }
 {
  bone1Name=base_cape_mid
  bone2Name=spine_02
  swing1LimitDegrees=45
  swing1Motion=Free
  swing2LimitDegrees=45
  swing2Motion=Free
  twistLimitDegrees=45
  twistMotion=Free
 }
 {
  bone1Name=spine_02
  bone2Name=spine_01
  swing1LimitDegrees=45
  swing1Motion=Free
  swing2LimitDegrees=45
  swing2Motion=Free
  twistLimitDegrees=45
  twistMotion=Free
 }
 {
  bone1Name=dyn_wing_upper_l_03
  bone2Name=spine_02
  swing1LimitDegrees=45
  swing1Motion=Free
  swing2LimitDegrees=45
  swing2Motion=Free
  twistLimitDegrees=45
  twistMotion=Free
 }
 {
  bone1Name=dyn_wing_upper_r_03
  bone2Name=spine_02
  swing1LimitDegrees=45
  swing1Motion=Free
  swing2LimitDegrees=45
  swing2Motion=Free
  twistLimitDegrees=45
  twistMotion=Free
 }
 {
  bone1Name=attach_backpack
  bone2Name=spine_02
  swing1LimitDegrees=45
  swing1Motion=Free
  swing2LimitDegrees=45
  swing2Motion=Free
  twistLimitDegrees=45
  twistMotion=Free
 }
 {
  bone1Name=dyn_cape_r_01
  bone2Name=dyn_cape_mid_01
  swing1LimitDegrees=45
  swing1Motion=Free
  swing2LimitDegrees=45
  swing2Motion=Free
  twistLimitDegrees=45
  twistMotion=Free
 }
 {
  bone1Name=dyn_cape_l_01
  bone2Name=dyn_cape_mid_01
  swing1LimitDegrees=45
  swing1Motion=Free
  swing2LimitDegrees=45
  swing2Motion=Free
  twistLimitDegrees=45
  twistMotion=Free
 }
 {
  bone1Name=dyn_cape_r_02
  bone2Name=dyn_cape_mid_02
  swing1LimitDegrees=45
  swing1Motion=Free
  swing2LimitDegrees=45
  swing2Motion=Free
  twistLimitDegrees=45
  twistMotion=Free
 }
 {
  bone1Name=dyn_cape_l_02
  bone2Name=dyn_cape_mid_02
  swing1LimitDegrees=45
  swing1Motion=Free
  swing2LimitDegrees=45
  swing2Motion=Free
  twistLimitDegrees=45
  twistMotion=Free
 }
 {
  bone1Name=dyn_cape_r_03
  bone2Name=dyn_cape_mid_03
  swing1LimitDegrees=45
  swing1Motion=Free
  swing2LimitDegrees=45
  swing2Motion=Free
  twistLimitDegrees=45
  twistMotion=Free
 }
 {
  bone1Name=dyn_cape_l_03
  bone2Name=dyn_cape_mid_03
  swing1LimitDegrees=45
  swing1Motion=Free
  swing2LimitDegrees=45
  swing2Motion=Free
  twistLimitDegrees=45
  twistMotion=Free
 }
 {
  bone1Name=dyn_wing_upper_l_04
  bone2Name=spine_02
  swing1LimitDegrees=45
  swing1Motion=Free
  swing2LimitDegrees=45
  swing2Motion=Free
  twistLimitDegrees=45
  twistMotion=Free
 }
 {
  bone1Name=dyn_wing_upper_r_04
  bone2Name=spine_02
  swing1LimitDegrees=45
  swing1Motion=Free
  swing2LimitDegrees=45
  swing2Motion=Free
  twistLimitDegrees=45
  twistMotion=Free
 }
]
