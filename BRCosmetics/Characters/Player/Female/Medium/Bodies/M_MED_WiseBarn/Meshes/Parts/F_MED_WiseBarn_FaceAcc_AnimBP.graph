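# Post-process style Anim Blueprint for the female WiseBarn face accessory, adding secondary motion to its hay, ribbon and medal bones.
# It copies the attached parent mesh's pose, then chains seven spring-driven, angle-limited Anim Dynamics nodes simulated root-relative.
# Each update computes a state-based gravity override and switches a velocity-scale variable on zipline enter/exit events.
# The AnimGraph shows no reads of those variables, using fixed node values instead.
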AnimBlueprint /BRCosmetics/Characters/Player/Female/Medium/Bodies/M_MED_WiseBarn/Meshes/Parts/F_MED_WiseBarn_FaceAcc_AnimBP
target_skeleton=&/F_MED_WiseBarn_FaceAcc_Skeleton
graphs [
 AnimGraph {nodes [
   n1 = Anim Dynamics - Bone: dyn_hay_01!(
    ComponentPose=Local To Component(LocalPose=Copy Pose From Mesh() {bUseAttachedParent=true})
    Alpha=1
    GravityOverride:Vector="0,0,0"
    ComponentLinearVelScale:Vector="0,0,0"
    ExternalForce="0,0,0"
   ) -> (Pose) {
    angularDampingOverride=1
    angularSpringConstant=11000
    bAngularSpring=true
    bOverrideAngularDamping=true
    bUseGravityOverride=true
    gravityScale=0.35
    simulationSpace=RootRelative
    boundBone {boneName=dyn_hay_01}
    componentAppliedLinearAccClamp=<9500 9500 9500>
    componentLinearAccScale=<0.5 0.5 0.5>
    componentLinearVelScale=<1 1 1>
    gravityOverride {z=-980}
    physicsBodyDefinitions [{
      boundBone {boneName=dyn_hay_01}
      boxExtents=<20 30 40>
      constraintSetup {angularLimitsMax=<5 10 5>, angularLimitsMin=<-5 -10 -10>, angularTarget=<1 0 0>}
      localJointOffset=<2 0 0>
     }]
   }
   n2 = Anim Dynamics - Bone: dyn_hay_02!(ComponentPose=@n1, Alpha=1, GravityOverride:Vector="0,0,0", ComponentLinearVelScale:Vector="0,0,0", ExternalForce="0,0,0") -> (Pose) {
    angularDampingOverride=1
    angularSpringConstant=20000
    bAngularSpring=true
    bOverrideAngularDamping=true
    bUseGravityOverride=true
    gravityScale=0.35
    simulationSpace=RootRelative
    boundBone {boneName=dyn_hay_02}
    componentLinearAccScale=<0.5 0.5 0.5>
    componentLinearVelScale=<1 1 1>
    gravityOverride {z=-980}
    physicsBodyDefinitions [{
      boundBone {boneName=dyn_hay_02}
      boxExtents=<20 30 40>
      constraintSetup {angularLimitsMax=<5 10 8>, angularLimitsMin=<-5 -10 -15>, angularTarget=<1 0 0>}
      localJointOffset=<3 0 0>
     }]
   }
   n3 = Anim Dynamics - Bone: dyn_hay_03!(ComponentPose=@n2, Alpha=1, GravityOverride:Vector="0,0,0", ComponentLinearVelScale:Vector="0,0,0", ExternalForce="0,0,0") -> (Pose) {
    angularDampingOverride=1
    angularSpringConstant=20000
    bAngularSpring=true
    bOverrideAngularDamping=true
    bUseGravityOverride=true
    gravityScale=0.35
    simulationSpace=RootRelative
    boundBone {boneName=dyn_hay_03}
    componentLinearAccScale=<1 1 1>
    componentLinearVelScale=<1 1 1>
    gravityOverride {z=-980}
    physicsBodyDefinitions [{
      boundBone {boneName=dyn_hay_03}
      boxExtents=<20 30 30>
      constraintSetup {angularLimitsMax=<5 5 8>, angularLimitsMin=<-5 -5 -8>, angularTarget=<1 0 0>}
      localJointOffset=<4 0 0>
     }]
   }
   n4 = Anim Dynamics - Bone: dyn_hay_04!(ComponentPose=@n3, Alpha=1, GravityOverride:Vector="0,0,0", ComponentLinearVelScale:Vector="0,0,0", ExternalForce="0,0,0") -> (Pose) {
    angularDampingOverride=1
    angularSpringConstant=20000
    bAngularSpring=true
    bOverrideAngularDamping=true
    bUseGravityOverride=true
    gravityScale=0.35
    simulationSpace=RootRelative
    boundBone {boneName=dyn_hay_04}
    componentLinearAccScale=<1 1 1>
    componentLinearVelScale=<1 1 1>
    gravityOverride {z=-980}
    physicsBodyDefinitions [{
      boundBone {boneName=dyn_hay_04}
      boxExtents=<20 30 30>
      constraintSetup {angularLimitsMax=<5 5 8>, angularLimitsMin=<-5 -5 -8>, angularTarget=<1 0 0>}
      localJointOffset=<4 0 0>
     }]
   }
   n5 = Anim Dynamics - Bone: dyn_ribbon_l!(ComponentPose=@n4, Alpha=1, GravityOverride:Vector="0,0,-980", ComponentLinearVelScale:Vector="1,1,1", ExternalForce="0,0,0") -> (Pose) {
    angularDampingOverride=1
    angularSpringConstant=11000
    bAngularSpring=true
    bOverrideAngularDamping=true
    bUseGravityOverride=true
    gravityScale=0.35
    simulationSpace=RootRelative
    boundBone {boneName=dyn_ribbon_l}
    componentAppliedLinearAccClamp=<9500 9500 9500>
    componentLinearAccScale=<0.5 0.5 0.5>
    componentLinearVelScale=<1 1 1>
    gravityOverride {z=-980}
    physicsBodyDefinitions [{
      boundBone {boneName=dyn_ribbon_l}
      boxExtents=<30 40 40>
      constraintSetup {angularLimitsMax=<1 1 15>, angularLimitsMin=<-1 -10 -15>, angularTarget=<1 0 0>}
      localJointOffset=<4 0 0>
     }]
   }
   n6 = Anim Dynamics - Bone: dyn_medal!(ComponentPose=@n5, Alpha=1, GravityOverride:Vector="0,0,-980", ComponentLinearVelScale:Vector="1,1,1", ExternalForce="0,0,0") -> (Pose) {
    angularDampingOverride=1
    angularSpringConstant=11000
    bAngularSpring=true
    bOverrideAngularDamping=true
    bUseGravityOverride=true
    gravityScale=0.35
    simulationSpace=RootRelative
    boundBone {boneName=dyn_medal}
    componentAppliedLinearAccClamp=<9500 9500 9500>
    componentLinearAccScale=<0.5 0.5 0.5>
    componentLinearVelScale=<1 1 1>
    gravityOverride {z=-980}
    physicsBodyDefinitions [{
      boundBone {boneName=dyn_medal}
      boxExtents=<30 40 40>
      constraintSetup {angularLimitsMax=<1 1 15>, angularLimitsMin=<-1 -10 -15>, angularTarget=<1 0 0>}
      localJointOffset=<4 0 0>
     }]
   }
   n7 = Anim Dynamics - Bone: dyn_ribbon_r!(ComponentPose=@n6, Alpha=1, GravityOverride:Vector="0,0,-980", ComponentLinearVelScale:Vector="1,1,1", ExternalForce="0,0,0") -> (Pose) {
    angularDampingOverride=1
    angularSpringConstant=11000
    bAngularSpring=true
    bOverrideAngularDamping=true
    bUseGravityOverride=true
    gravityScale=0.35
    simulationSpace=RootRelative
    boundBone {boneName=dyn_ribbon_r}
    componentAppliedLinearAccClamp=<9500 9500 9500>
    componentLinearAccScale=<0.5 0.5 0.5>
    componentLinearVelScale=<1 1 1>
    gravityOverride {z=-980}
    physicsBodyDefinitions [{
      boundBone {boneName=dyn_ribbon_r}
      boxExtents=<30 40 40>
      constraintSetup {angularLimitsMax=<1 10 15>, angularLimitsMin=<-1 -1 -15>, angularTarget=<1 0 0>}
      localJointOffset=<4 0 0>
     }]
   }
   Output Pose(Result=Component To Local(ComponentPose=@n7)) {layerGroup=None}
  ]}
 EventGraph {nodes [
   Event BlueprintUpdateAnimation()
   > n1 = Sequence()
   n2 = Event BlueprintProcessAnimInputEvent()
   n3 = IsNamed(self=@n2.AnimInputEvent, InEventName=aie.Player.Zipline.Entered)
   n4 = OR Boolean(A=Get bIsMoving2D(), B=Get bIsTacticalSprinting())
   n5 = Get bIsJumping()
   n6 = Branch(execute=@n1.then_1, Condition=OR Boolean(A=@n3, B=AND Boolean!(A=@n3, B=@n5, C:bool=@n4) -> (ReturnValue)))
   > Set Componen Linear Vel Scale(Componen Linear Vel Scale:Vector="0, 0, 0") -> (Output_Get:Vector)
   Branch(execute=@n6.else, Condition=IsNamed(self=@n2.AnimInputEvent, InEventName=aie.Player.Zipline.Exited))
   > Set Componen Linear Vel Scale(Componen Linear Vel Scale="0.4,0.4,0.4")
   n7 = "Gravity Override State Function (Nativized)"!(
    execute=@n1.then_0
    CharacterPartAnimInstance:Custom Character Part Anim Instance Object Reference=Self-Reference()
    GravityOverrideParameterSet:Fort Gravity Override Parameter Set Structure=Get Gravity Override Params_0() -> (Gravity Override Params_0:Fort Gravity Override Parameter Set Structure)
    RigidBodyAnimNodeParameterSet:Fort Rigid Body Anim Node Parameter Set Structure=Get RBAN Params_0() -> (RBAN Params_0:Fort Rigid Body Anim Node Parameter Set Structure)
    DebugDraw:bool=false
    GameTimeInSeconds:float=GetGameTimeInSeconds()
    DebugJointName:Name=root
    StateLogicLayerName:Name=none
   ) -> (
    then
    OutGravityOverride:Vector
    OutCurrentState:Name=None
    OutGravityOverrideParameters:Fort Gravity Override Parameters Structure
    OutRigidBodyAnimNodeParameters:Fort Rigid Body Anim Node Parameters Structure
   )
   > Set GravityOverride(GravityOverride=@n7)
   > Set Sim Space Settings(Sim Space Settings:Sim Space Settings Structure) -> (Output_Get:Sim Space Settings Structure)
  ]}
]
variables [
 Componen Linear Vel Scale {type=Vector, default=<0.4 0.4 0.4>}
 Gravity Override Params_0 {
  type=FortGravityOverrideParameterSet
  default=(Skydive_GoingUp=(GravityOverride=()),Skydive_FloatingDown=(GravityOverride=()),Skydive_DivingDown=(GravityOverride=()),Skydive_Parachuting=(GravityOverride=()),OnGround_Standing=(GravityOverride=()),OnGround_Moving=(GravityOverride=()),OnGround_Moving_HighRotationRate=(GravityOverride=()),OnGround_Crouch_Moving=(GravityOverride=()),Emote_or_Melee=(GravityOverride=()),Swimming=(GravityOverride=()),Front_End=(GravityOverride=()),DBNO=(GravityOverride=()),DBNO_Carried=(GravityOverride=()),OnGround_Falling=(GravityOverride=()),Ziplining=(GravityOverride=()),Skydive_Surfing=(GravityOverride=()),Riding_Standing=(GravityOverride=()),Riding_Moving=(GravityOverride=()),Riding_Moving_Targeting=(GravityOverride=()),Tactical_Sprint=(GravityOverride=()),Rail_Grinding=(GravityOverride=()),)
 }
 GravityOverride {type=Vector}
 RBAN Params_0 {
  type=FortRigidBodyAnimNodeParameterSet
  default=(Skydive_GoingUp=(RigidBodyAnimNode=()),Skydive_FloatingDown=(RigidBodyAnimNode=()),Skydive_DivingDown=(RigidBodyAnimNode=()),Skydive_Parachuting=(RigidBodyAnimNode=()),OnGround_Standing=(RigidBodyAnimNode=()),OnGround_Moving=(RigidBodyAnimNode=()),OnGround_Moving_HighRotationRate=(RigidBodyAnimNode=()),OnGround_Crouch_Moving=(RigidBodyAnimNode=()),Emote_or_Melee=(RigidBodyAnimNode=()),Swimming=(RigidBodyAnimNode=()),Front_End=(RigidBodyAnimNode=()),DBNO=(RigidBodyAnimNode=()),DBNO_Carried=(RigidBodyAnimNode=()),OnGround_Falling=(RigidBodyAnimNode=()),Ziplining=(RigidBodyAnimNode=()),Skydive_Surfing=(RigidBodyAnimNode=()),Riding_Standing=(RigidBodyAnimNode=()),Riding_Moving=(RigidBodyAnimNode=()),Riding_Moving_Targeting=(RigidBodyAnimNode=()),Tactical_Sprint=(RigidBodyAnimNode=()),Rail_Grinding=(RigidBodyAnimNode=()),)
 }
 Sim Space Settings {
  type=SimSpaceSettings
  default {
   dampingAlpha=1
   maxAngularAcceleration=10000
   maxAngularVelocity=10000
   maxLinearAcceleration=10000
   maxLinearVelocity=10000
   velocityScaleZ=1
   worldAlpha=0
   externalLinearDragV=<0 0 0>
  }
 }
]
